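# Iridescent Default Lit material for the Creative golf ball toy, with color that varies by view angle.
# Normal-dot-camera plus the camera vector, added for variance on the small ball, drives a hue shift, desaturated to index a curve atlas gradient.
# That gradient tints the base color, then mask G applies the F shadow and mask R the F color.
# The pattern blend and time-based pan nodes feed no other node.
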Material /BRCosmetics/Creative/Toys/GolfBall/Materials/M_CP_GolfBall_Iridescence_01
graph {
 properties {pixelDepthOffsetMode=PDOM_Legacy, usedShadingModels=Default Lit}
 nodes [
  n1 = DotProduct(A=PixelNormalWS(), B=CameraVectorWS())
  AppendVector(A=@n1, B=@n1)
  n2 = VectorParameter() {group=Color, parameterName=Iridescence Base Color, sortPriority=2, defaultValue=<rgba 1 1 1 1>}
  n3 = TextureSampleParameter2D(UVs=Multiply(A=TextureCoordinate(), B=ScalarParameter() {defaultValue=1, group=Color, parameterName=Pattern Tiling})) {
   group=Color
   parameterName=Pattern Texture
   sortPriority=5
   texture=/BRCosmetics/Creative/Toys/Basketball/Textures/SportingGoods_Basketball_Creative_Pattern_TriangleBars_01
  }
  LinearInterpolate(
   A=@n2.RGB
   B=VectorParameter() {group=Color, parameterName=Iridescence Pattern Color, sortPriority=2, defaultValue=<rgba 1 1 1 1>}.RGB
   Alpha=StaticSwitchParameter(True=OneMinus(None=@n3.RGB), False=@n3.RGB) {group=Color, parameterName=Invert Pattern, sortPriority=6}
  )
  n4 = Add(A=@n1, B=CameraVectorWS() {desc=To Add More Variance Due to Small Size})
  n5 = VectorParameter() {group=Hue Shift, parameterName=Hue Shift Values, sortPriority=3, defaultValue=<rgba 1 1 1 1>}
  n6 = CurveAtlasRowParameter(CurveTime=Desaturation(None=MaterialFunctionCall(Hue Shift Percentage=@n4, Texture=@n5.RGB) {materialFunction=&HueShift}.Result)) {
   atlas=&AthenaCurveAtlas01
   curve=/Game/Athena/Environments/Nexus/Materials/FXshader/Curve_Aix
   defaultValue=77
   group=Hue Shift
   parameterName=Curve
   sortPriority=2
  }
  n7 = LinearInterpolate(
   B=Multiply(A=@n6, B=ScalarParameter() {defaultValue=1, group=Hue Shift, parameterName=Hue Shift Brightness, sortPriority=5})
   Alpha=ScalarParameter() {defaultValue=1, group=Hue Shift, parameterName=Hue Shift Intensity, sortPriority=4}
  ) {constA=1}
  n8 = VectorParameter() {group=Color, parameterName=F Shadow Color, sortPriority=2, defaultValue=<rgba 0 0 0 1>}
  n9 = TextureSample() {texture=/BRCosmetics/Creative/Toys/GolfBall/Textures/T_SportingGoods_GolfBall_Creative_Masks_01}
  LinearInterpolate(
   A=Multiply(A=Multiply(A=@n7, B=@n2.RGB), B=LinearInterpolate(B=@n8.RGB, Alpha=@n9.G) {constA=1})
   B=VectorParameter() {group=Color, parameterName=F Color, sortPriority=2, defaultValue=<rgba 1 1 1 1>}.RGB
   Alpha=@n9.R
  )
  Multiply(A=Time(), B=ScalarParameter() {group=Hue Shift, parameterName=Pan Speed})
  TextureSample() {samplerType=SAMPLERTYPE_Normal, texture=/BRCosmetics/Creative/Toys/GolfBall/Textures/T_SportingGoods_GolfBall_Creative_01_N}
  TextureSample() {texture=/BRCosmetics/Creative/Toys/GolfBall/Textures/T_SportingGoods_GolfBall_Creative_01_S}
 ]
}
